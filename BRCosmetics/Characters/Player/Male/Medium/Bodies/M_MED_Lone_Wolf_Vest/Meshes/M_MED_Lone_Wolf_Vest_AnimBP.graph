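# Anim Blueprint for the Lone Wolf Vest cosmetic body, adding secondary motion to skirt, collar, katanas, zipper and arm pads.
# The parent's copied pose passes through a Control Rig, corrective bone offsets driven by limb/head rotations, and AnimDynamics into a cached pose.
# The skirt uses crouch-switched RigidBody simulation, or procedural bone chains (trails, look-ats, DBNO variant) without it.
# Events toggle a material element by Lone Wolf pickaxe and skydiving state.
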AnimBlueprint /BRCosmetics/Characters/Player/Male/Medium/Bodies/M_MED_Lone_Wolf_Vest/Meshes/M_MED_Lone_Wolf_Vest_AnimBP
target_skeleton=&/M_MED_Lone_Wolf_Vest_Skeleton
graphs [
 AnimGraph {nodes [
   n1 = "Transform (Modify) Bone - Bone: dyn_skirt_bk_l_MSTR"(
    ComponentPose=Local To Component(LocalPose=Use cached pose 'prePose'())
    Alpha=Get thigh_rot_bk_l() -> (thigh_rot_bk_l:double)
    Translation="0,0,0"
    Rotation="30,-20,0"
    Scale="1,1,1"
   ) {
    rotationMode=BMM_Additive
    rotationSpace=BCS_BoneSpace
    translationMode=BMM_Additive
    translationSpace=BCS_BoneSpace
    alphaBoolBlend {blendOption=Linear}
    boneToModify {boneName=dyn_skirt_bk_l_MSTR}
    rotation {pitch=30, yaw=-20}
   }
   n2 = "Transform (Modify) Bone - Bone: dyn_skirt_bk_r_MSTR"(ComponentPose=@n1, Alpha=Get thigh_rot_bk_r() -> (thigh_rot_bk_r:double), Translation="0,0,0", Rotation="30,0,0", Scale="1,1,1") {
    rotationMode=BMM_Additive
    rotationSpace=BCS_BoneSpace
    translationMode=BMM_Additive
    translationSpace=BCS_BoneSpace
    alphaBoolBlend {blendOption=Linear}
    boneToModify {boneName=dyn_skirt_bk_r_MSTR}
    rotation {pitch=30}
   }
   n3 = "Transform (Modify) Bone - Bone: dyn_skirt_bk_l_MSTR"(
    ComponentPose=@n2
    Alpha=Get thigh_rot_out_l() -> (thigh_rot_out_l:double)
    Translation="0,-0.5,1"
    Rotation="11.161,-0.000,-5.471"
    Scale="1,1,1"
   ) {
    rotationMode=BMM_Additive
    rotationSpace=BCS_BoneSpace
    translationMode=BMM_Additive
    translationSpace=BCS_BoneSpace
    alphaBoolBlend {blendOption=Linear}
    boneToModify {boneName=dyn_skirt_bk_l_MSTR}
    rotation {pitch=11.161, roll=-5.471}
    translation {y=-0.5, z=1}
   }
   n4 = "Transform (Modify) Bone - Bone: dyn_skirt_bk_r_MSTR"(
    ComponentPose=@n3
    Alpha=Get thigh_rot_out_r() -> (thigh_rot_out_r:double)
    Translation="0,0.5,0.751"
    Rotation="11.161,-0.000,5.471"
    Scale="1,1,1"
   ) {
    rotationMode=BMM_Additive
    rotationSpace=BCS_BoneSpace
    translationMode=BMM_Additive
    translationSpace=BCS_BoneSpace
    alphaBoolBlend {blendOption=Linear}
    boneToModify {boneName=dyn_skirt_bk_r_MSTR}
    rotation {pitch=11.161, roll=5.471}
    translation {y=0.5, z=0.751}
   }
   n5 = "Constraint - dyn_skirt_bk_base (PP)"!(ComponentPose=@n4, Alpha=1, ConstraintWeights_0=0.5, ConstraintWeights_1:float=0.5) -> (Pose) {
    alphaBoolBlend {blendOption=Linear}
    boneToModify {boneName=dyn_skirt_bk_base}
    constraintSetup [{targetBone {boneName=dyn_skirt_bk_l_MSTR}}, {targetBone {boneName=dyn_skirt_bk_r_MSTR}}]
    constraintWeights [1,1]
   }
   n6 = Get bIsCrouching()
   n7 = Look At - Bone: dyn_skirt_bk_l_02!(ComponentPose=@n5, bAlphaBoolEnabled:bool=@n6) -> (Pose) {
    alphaInputType=Bool
    alphaBoolBlend {blendInTime=0.3, blendOption=Linear}
    boneToModify {boneName=dyn_skirt_bk_l_02}
    lookAtLocation=<35 -45 -50>
    lookAtTarget {boneReference {boneName=root}}
    lookAt_Axis {axis=<1 0>}
   }
   n8 = Look At - Bone: dyn_skirt_bk_02!(ComponentPose=@n7, bAlphaBoolEnabled:bool=@n6) -> (Pose) {
    alphaInputType=Bool
    alphaBoolBlend {blendInTime=0.3, blendOption=Linear}
    boneToModify {boneName=dyn_skirt_bk_02}
    lookAtLocation=<0 -70 -50>
    lookAtTarget {boneReference {boneName=root}}
    lookAt_Axis {axis=<1 0>}
   }
   n9 = Look At - Bone: dyn_skirt_bk_r_02!(ComponentPose=@n8, bAlphaBoolEnabled:bool=@n6) -> (Pose) {
    alphaInputType=Bool
    alphaBoolBlend {blendInTime=0.3, blendOption=Linear}
    boneToModify {boneName=dyn_skirt_bk_r_02}
    lookAtLocation=<-10 -70 -50>
    lookAtTarget {boneReference {boneName=root}}
    lookAt_Axis {axis=<1 0>}
   }
   n10 = Get backLoin_deform_alpha() -> (backLoin_deform_alpha:double)
   n11 = "Transform (Modify) Bone - Bone: dyn_skirt_bk_r_01"(ComponentPose=@n9, Alpha=@n10, Translation="0,2,0", Rotation="0,10,0", Scale="1,1,1") {
    rotationMode=BMM_Additive
    rotationSpace=BCS_BoneSpace
    scaleMode=BMM_Additive
    scaleSpace=BCS_BoneSpace
    translationMode=BMM_Additive
    translationSpace=BCS_BoneSpace
    alphaBoolBlend {blendOption=Linear}
    alphaScaleBiasClamp {bInterpResult=true, interpSpeedDecreasing=5}
    boneToModify {boneName=dyn_skirt_bk_r_01}
    rotation {yaw=10}
    translation {y=2}
   }
   n12 = "Transform (Modify) Bone - Bone: dyn_skirt_bk_r_02"(ComponentPose=@n11, Alpha=@n10, Translation="0,0,2", Rotation="-20,15,0", Scale="1,1,1") {
    rotationMode=BMM_Additive
    rotationSpace=BCS_BoneSpace
    scaleMode=BMM_Additive
    scaleSpace=BCS_BoneSpace
    translationMode=BMM_Additive
    translationSpace=BCS_BoneSpace
    alphaBoolBlend {blendOption=Linear}
    alphaScaleBiasClamp {bInterpResult=true, interpSpeedDecreasing=5}
    boneToModify {boneName=dyn_skirt_bk_r_02}
    rotation {pitch=-20, yaw=15}
    translation {z=2}
   }
   n13 = "Transform (Modify) Bone - Bone: dyn_skirt_bk_02"(ComponentPose=@n12, Alpha=@n10, Translation="0,0,0", Rotation="-15,30,0", Scale="1,1,1") {
    rotationMode=BMM_Additive
    rotationSpace=BCS_BoneSpace
    scaleMode=BMM_Additive
    scaleSpace=BCS_BoneSpace
    translationMode=BMM_Additive
    translationSpace=BCS_BoneSpace
    alphaBoolBlend {blendOption=Linear}
    alphaScaleBiasClamp {bInterpResult=true, interpSpeedDecreasing=5}
    boneToModify {boneName=dyn_skirt_bk_02}
    rotation {pitch=-15, yaw=30}
   }
   n14 = Get BwdMovement() -> (BwdMovement:double)
   n15 = Trail controller - Bone: dyn_skirt_bk_l_03(ComponentPose=@n13, Alpha=@n14) {
    chainLength=3
    maxDeltaTime=0.0333
    alphaScaleBiasClamp {bInterpResult=true, interpSpeedIncreasing=5}
    planarLimits [{drivingBone {boneName=pelvis}, planeTransform {location=<0 -11 0>, rotation=<rot 0 -5.0001 -89.9999>, scale=<0.5 0.5 0.5>}}]
    rotationLimits [{limitMax=<180 180 180>}, {limitMax=<180 180 180>}, {limitMax=<180 180 180>}]
    rotationOffsets [<0 0 0>, <0 0 0>, <0 0 0>]
    trailBone {boneName=dyn_skirt_bk_l_03}
    trailRelaxationSpeed {editorCurveData {keys [<at 0 20 RCIM_Linear>,<at 0.5055 25 RCIM_Linear>,<at 1 5 RCIM_Linear>]}}
   }
   n16 = Trail controller - Bone: dyn_skirt_bk_r_03(ComponentPose=@n15, Alpha=@n14) {
    chainLength=3
    maxDeltaTime=0.0333
    alphaScaleBiasClamp {bInterpResult=true, interpSpeedIncreasing=5}
    planarLimits [{drivingBone {boneName=pelvis}, planeTransform {location=<0 -11 0>, rotation=<rot 0 -5.0001 -89.9999>, scale=<0.5 0.5 0.5>}}]
    rotationLimits [{limitMax=<180 180 180>}, {limitMax=<180 180 180>}, {limitMax=<180 180 180>}]
    rotationOffsets [<0 0 0>, <0 0 0>, <0 0 0>]
    trailBone {boneName=dyn_skirt_bk_r_03}
    trailRelaxationSpeed {editorCurveData {keys [<at 0 25 RCIM_Linear>,<at 0.5055 25 RCIM_Linear>,<at 1 5 RCIM_Linear>]}}
   }
   n17 = Trail controller - Bone: dyn_skirt_bk_03(ComponentPose=@n16, Alpha=@n14) {
    chainLength=3
    maxDeltaTime=0.0333
    alphaScaleBiasClamp {bInterpResult=true, interpSpeedIncreasing=5}
    planarLimits [{drivingBone {boneName=pelvis}, planeTransform {location=<0 -11 0>, rotation=<rot 0 -5.0001 -89.9999>, scale=<0.5 0.5 0.5>}}]
    rotationLimits [{limitMax=<180 180 180>}, {limitMax=<180 180 180>}, {limitMax=<180 180 180>}]
    rotationOffsets [<0 0 0>, <0 0 0>, <0 0 0>]
    trailBone {boneName=dyn_skirt_bk_03}
    trailRelaxationSpeed {editorCurveData {keys [<at 0 25 RCIM_Linear>,<at 0.5055 25 RCIM_Linear>,<at 1 5 RCIM_Linear>]}}
   }
   n18 = Component To Local(ComponentPose=@n17)
   n19 = "Transform (Modify) Bone - Bone: dyn_skirt_bk_l_01"!(
    ComponentPose=Local To Component(LocalPose=Use cached pose 'prePose'())
    bAlphaBoolEnabled:bool=True
    Translation="0,0,0"
    Rotation="-25,25,0"
    Scale="1,1,1"
   ) -> (Pose) {
    alphaInputType=Bool
    rotationMode=BMM_Additive
    rotationSpace=BCS_BoneSpace
    translationMode=BMM_Additive
    translationSpace=BCS_BoneSpace
    alphaBoolBlend {blendOption=Linear}
    boneToModify {boneName=dyn_skirt_bk_l_01}
    rotation {pitch=-25, yaw=25}
   }
   n20 = "Transform (Modify) Bone - Bone: dyn_skirt_bk_l_02"!(ComponentPose=@n19, bAlphaBoolEnabled:bool=True, Translation="0,0,2", Rotation="-15,15,0.000", Scale="1,1,1") -> (Pose) {
    alphaInputType=Bool
    rotationMode=BMM_Additive
    rotationSpace=BCS_BoneSpace
    scaleMode=BMM_Additive
    scaleSpace=BCS_BoneSpace
    translationMode=BMM_Additive
    translationSpace=BCS_BoneSpace
    alphaBoolBlend {blendOption=Linear}
    boneToModify {boneName=dyn_skirt_bk_l_02}
    rotation {pitch=-15, yaw=15}
    translation {z=2}
   }
   n21 = "Transform (Modify) Bone - Bone: dyn_skirt_bk_r_01"!(ComponentPose=@n20, bAlphaBoolEnabled:bool=True, Translation="0,0,0", Rotation="-25,-25,0", Scale="1,1,1") -> (Pose) {
    alphaInputType=Bool
    rotationMode=BMM_Additive
    rotationSpace=BCS_BoneSpace
    translationMode=BMM_Additive
    translationSpace=BCS_BoneSpace
    alphaBoolBlend {blendOption=Linear}
    boneToModify {boneName=dyn_skirt_bk_r_01}
    rotation {pitch=-25, yaw=-25}
   }
   n22 = "Transform (Modify) Bone - Bone: dyn_skirt_bk_r_02"!(ComponentPose=@n21, bAlphaBoolEnabled:bool=True, Translation="0,0,2", Rotation="-15,-15,0.000", Scale="1,1,1") -> (Pose) {
    alphaInputType=Bool
    rotationMode=BMM_Additive
    rotationSpace=BCS_BoneSpace
    scaleMode=BMM_Additive
    scaleSpace=BCS_BoneSpace
    translationMode=BMM_Additive
    translationSpace=BCS_BoneSpace
    alphaBoolBlend {blendOption=Linear}
    boneToModify {boneName=dyn_skirt_bk_r_02}
    rotation {pitch=-15, yaw=-15}
    translation {z=2}
   }
   n23 = "Transform (Modify) Bone - Bone: dyn_skirt_bk_01"!(ComponentPose=@n22, bAlphaBoolEnabled:bool=True, Translation="0,0,0", Rotation="-25,0,0", Scale="1,1,1") -> (Pose) {
    alphaInputType=Bool
    rotationMode=BMM_Additive
    rotationSpace=BCS_BoneSpace
    translationMode=BMM_Additive
    translationSpace=BCS_BoneSpace
    alphaBoolBlend {blendOption=Linear}
    boneToModify {boneName=dyn_skirt_bk_01}
    rotation {pitch=-25}
   }
   n24 = "Transform (Modify) Bone - Bone: dyn_skirt_bk_02"!(ComponentPose=@n23, bAlphaBoolEnabled:bool=True, Translation="0,0,0", Rotation="-30,0,0.000", Scale="1,1,1") -> (Pose) {
    alphaInputType=Bool
    rotationMode=BMM_Additive
    rotationSpace=BCS_BoneSpace
    scaleMode=BMM_Additive
    scaleSpace=BCS_BoneSpace
    translationMode=BMM_Additive
    translationSpace=BCS_BoneSpace
    alphaBoolBlend {blendOption=Linear}
    boneToModify {boneName=dyn_skirt_bk_02}
    rotation {pitch=-30}
   }
   n25 = Look At - Bone: dyn_skirt_bk_l_01(ComponentPose=@n24, Alpha=1) {
    alphaBoolBlend {blendOption=Linear}
    boneToModify {boneName=dyn_skirt_bk_l_01}
    lookAtLocation {x=0}
    lookAtTarget {boneReference {boneName=foot_l}}
    lookAt_Axis {axis=<1 0>}
   }
   n26 = Look At - Bone: dyn_skirt_bk_r_01(ComponentPose=@n25, Alpha=1) {
    alphaBoolBlend {blendOption=Linear}
    boneToModify {boneName=dyn_skirt_bk_r_01}
    lookAtLocation {x=0}
    lookAtTarget {boneReference {boneName=foot_r}}
    lookAt_Axis {axis=<1 0>}
   }
   n27 = "Constraint - dyn_skirt_bk_01 (PPP)"!(ComponentPose=@n26, Alpha=1, ConstraintWeights_0=0.4, ConstraintWeights_1:float=0.4, ConstraintWeights_2:float=0.2) -> (Pose) {
    alphaBoolBlend {blendOption=Linear}
    boneToModify {boneName=dyn_skirt_bk_01}
    constraintSetup [{targetBone {boneName=dyn_skirt_bk_l_01}}, {targetBone {boneName=dyn_skirt_bk_r_01}}, {targetBone {boneName=dyn_skirt_bk_base}}]
    constraintWeights [1,1,1]
   }
   n28 = Component To Local(ComponentPose=@n27)
   n29 = Get bIsDBNO()
   n30 = "Transform (Modify) Bone - Bone: dyn_skirt_bk_r_MSTR"(
    ComponentPose=Local To Component(LocalPose=Use cached pose 'prePose'())
    Alpha=Get backLoin_deform_alpha()
    Translation="0,3,2"
    Rotation="0,0,0"
    Scale="1,1,1"
   ) {
    rotationMode=BMM_Additive
    rotationSpace=BCS_BoneSpace
    scaleMode=BMM_Additive
    scaleSpace=BCS_BoneSpace
    translationMode=BMM_Additive
    translationSpace=BCS_BoneSpace
    alphaBoolBlend {blendOption=Linear}
    boneToModify {boneName=dyn_skirt_bk_r_MSTR}
    translation {y=3, z=2}
   }
   n31 = Get GravityOverride()
   n32 = RigidBody!(ComponentPose=@n30, Alpha, OverrideWorldGravity:Vector=@n31, ExternalForce) -> (Pose) {
    bClampLinearTranslationLimitToRefPose=true
    bOverrideWorldGravity=true
    overridePhysicsAsset=/BRCosmetics/Characters/Player/Male/Medium/Bodies/M_MED_Lone_Wolf/Meshes/M_MED_Lone_Wolf_dyn_skirt_crouching
    simulationSpace=BaseBoneSpace
    alphaBoolBlend {blendOption=Linear}
    baseBoneRef {boneName=spine_01}
    componentAppliedLinearAccClamp=<3000 3000 3000>
    componentLinearAccScale {x=0.45}
    componentLinearVelScale {x=0.45}
    overrideWorldGravity {z=-980}
    simSpaceSettings {
     maxAngularAcceleration=200
     maxAngularVelocity=5
     maxLinearAcceleration=1000
     maxLinearVelocity=200
     velocityScaleZ=0
     worldAlpha=1
     externalLinearDragV=<1 1>
    }
   }
   n33 = Component To Local(ComponentPose=@n32)
   n34 = RigidBody(ComponentPose=Local To Component(LocalPose=Use cached pose 'prePose'()), OverrideWorldGravity=@n31) {
    bClampLinearTranslationLimitToRefPose=true
    bOverrideWorldGravity=true
    overridePhysicsAsset=/BRCosmetics/Characters/Player/Male/Medium/Bodies/M_MED_Lone_Wolf/Meshes/M_MED_Lone_Wolf_dyn_skirt
    simulationSpace=BaseBoneSpace
    alphaBoolBlend {blendOption=Linear}
    baseBoneRef {boneName=spine_01}
    componentAppliedLinearAccClamp=<3000 3000 3000>
    componentLinearAccScale {x=0.45}
    componentLinearVelScale {x=0.45}
    overrideWorldGravity {z=-980}
    simSpaceSettings {
     maxAngularAcceleration=200
     maxAngularVelocity=5
     maxLinearAcceleration=1000
     maxLinearVelocity=200
     velocityScaleZ=0
     worldAlpha=1
     externalLinearDragV=<1 1>
    }
   }
   n35 = Component To Local(ComponentPose=@n34)
   n36 = Get bIsCrouching()
   Output Pose(Result=Blend Poses by bool(
     BlendPose_0=Blend Poses by bool(BlendPose_0=@n33, BlendPose_1=@n35, bActiveValue=@n36) {blendType=Linear}
     BlendPose_1=Blend Poses by bool(BlendPose_0=@n28, BlendPose_1=@n18, bActiveValue=@n29) {blendType=Linear}
     bActiveValue=Get bRigidBodyEnabled()
    ) {blendType=Linear}) {layerGroup=None}
   n37 = Control Rig!(
    Source=Copy Pose From Mesh() {bUseAttachedParent=true}
    Alpha
    KneeRot_fr_l_114deg:Vector
    KneeTrans_fr_l_114deg:Vector="0.258,-3.591,0.009"
    KneeScale_fr_l_114deg:Vector="1,1,1"
    KneeScale_fr_l_130deg:Vector="1,1,1"
    KneeRot_fr_l_130deg:Vector="0,0,-51.2"
    KneeTrans_fr_l_130deg:Vector="0.258,-3.591,0.009"
   ) -> (Pose) {
    alphaBoolBlend {blendOption=Linear}
    controlRigAssetReference {blueprintRigClass=&/M_MED_Lone_Wolf_Vest_Deform_CtrlRig_sculpters.M_MED_Lone_Wolf_Vest_Deform_CtrlRig_sculpters_C}
   }
   n38 = "Constraint - dyn_katanas_MSTR (PPP)"!(
    ComponentPose=Local To Component(LocalPose=@n37)
    Alpha=1
    ConstraintWeights_0=0.7218
    ConstraintWeights_1:float=0.1811
    ConstraintWeights_2:float=0.0971
   ) -> (Pose) {
    alphaBoolBlend {blendOption=Linear}
    boneToModify {boneName=dyn_katanas_MSTR}
    constraintSetup [{targetBone {boneName=pelvis}}, {targetBone {boneName=spine_01}}, {targetBone {boneName=dyn_skirt_ft_r_MSTR}}]
    constraintWeights [1,1,1]
   }
   n39 = Anim Dynamics - Bone: dyn_katanas_base(ComponentPose=@n38, Alpha=1, ExternalForce="0,0,0") {
    angularDampingOverride=0.999
    angularSpringConstant=2000
    bGravityOverrideInSimSpace=true
    bOverrideAngularDamping=true
    bUseGravityOverride=true
    gravityScale=0.35
    simulationSpace=RootRelative
    alphaBoolBlend {blendOption=Linear}
    boundBone {boneName=dyn_katanas_base}
    componentLinearAccScale=<0.2 0.2 0.2>
    componentLinearVelScale=<0.2 0.2 0.2>
    gravityOverride {z=-980}
    physicsBodyDefinitions [{
      sphereCollisionRadius=0
      boundBone {boneName=dyn_katanas_base}
      boxExtents=<10 15 20>
      constraintSetup {angularLimitsMax=<0 0 5>, angularLimitsMin=<0 0 -5>, angularTarget=<0 0 0>}
      localJointOffset=<1 0 0>
     }]
    relativeSpaceBone {boneName=dyn_katanas_MSTR}
   }
   n40 = Anim Dynamics - Bone: dyn_katanas_02(ComponentPose=@n39, Alpha=1, ExternalForce="0,0,0") {
    angularDampingOverride=0.999
    angularSpringConstant=1000
    bAngularSpring=true
    bGravityOverrideInSimSpace=true
    bOverrideAngularDamping=true
    bUseGravityOverride=true
    gravityScale=0.35
    simulationSpace=RootRelative
    alphaBoolBlend {blendOption=Linear}
    boundBone {boneName=dyn_katanas_02}
    componentLinearAccScale=<0.2 0.2 0.2>
    componentLinearVelScale=<0.2 0.2 0.2>
    gravityOverride {z=-980}
    physicsBodyDefinitions [{
      sphereCollisionRadius=0
      boundBone {boneName=dyn_katanas_02}
      boxExtents=<10 15 20>
      constraintSetup {angularLimitsMax=<0 2 5.03>, angularLimitsMin=<0 -2 -3>, angularTarget=<0 0 0>}
      localJointOffset=<2 0 0>
     }]
    relativeSpaceBone {boneName=dyn_katanas_MSTR}
   }
   n41 = Anim Dynamics - Bone: dyn_katanas_01(ComponentPose=@n40, Alpha=1, ExternalForce="0,0,0") {
    angularDampingOverride=0.999
    angularSpringConstant=1000
    bAngularSpring=true
    bGravityOverrideInSimSpace=true
    bOverrideAngularDamping=true
    bUseGravityOverride=true
    gravityScale=0.35
    simulationSpace=RootRelative
    alphaBoolBlend {blendOption=Linear}
    boundBone {boneName=dyn_katanas_01}
    componentLinearAccScale=<0.2 0.2 0.2>
    componentLinearVelScale=<0.2 0.2 0.2>
    gravityOverride {z=-980}
    physicsBodyDefinitions [{
      sphereCollisionRadius=0
      boundBone {boneName=dyn_katanas_01}
      boxExtents=<10 15 20>
      constraintSetup {angularLimitsMax=<0 2 5.03>, angularLimitsMin=<0 -2 -3>, angularTarget=<0 0 0>}
      localJointOffset=<1 0 0>
     }]
    relativeSpaceBone {boneName=dyn_katanas_MSTR}
   }
   n42 = Get thigh_rot_up_l() -> (thigh_rot_up_l:double)
   n43 = "Transform (Modify) Bone - Bone: dyn_skirt_ft_l_MSTR"(ComponentPose=@n41, Alpha=@n42, Translation="-6.5,0.000,-0.000", Rotation="0,0,0", Scale="1,1,1") {
    translationMode=BMM_Additive
    translationSpace=BCS_BoneSpace
    alphaBoolBlend {blendOption=Linear}
    boneToModify {boneName=dyn_skirt_ft_l_MSTR}
    translation {x=-6.5}
   }
   n44 = Get thigh_rot_up_r() -> (thigh_rot_up_r:double)
   n45 = "Transform (Modify) Bone - Bone: dyn_skirt_ft_r_MSTR"(ComponentPose=@n43, Alpha=@n44, Translation="-4,0.000,-0.000", Rotation="0,0,0", Scale="1,1,1") {
    translationMode=BMM_Additive
    translationSpace=BCS_BoneSpace
    alphaBoolBlend {blendOption=Linear}
    boneToModify {boneName=dyn_skirt_ft_r_MSTR}
    translation {x=-4}
   }
   n46 = Look At - Bone: dyn_skirt_ft_l_AIM(ComponentPose=@n45, Alpha=0.9) {
    alphaBoolBlend {blendOption=Linear}
    boneToModify {boneName=dyn_skirt_ft_l_AIM}
    lookAtLocation {x=0}
    lookAtTarget {boneReference {boneName=dyn_skirt_ft_l_TGT}}
    lookAt_Axis {axis=<1 0>}
   }
   n47 = Look At - Bone: dyn_skirt_ft_r_AIM(ComponentPose=@n46, Alpha=1) {
    alphaBoolBlend {blendOption=Linear}
    boneToModify {boneName=dyn_skirt_ft_r_AIM}
    lookAtLocation {x=0}
    lookAtTarget {boneReference {boneName=dyn_skirt_ft_r_TGT}}
    lookAt_Axis {axis=<1 0>}
   }
   n48 = "Transform (Modify) Bone - Bone: dyn_skirt_ft_l_AIM"(ComponentPose=@n47, Alpha=@n42, Translation="0,0,0", Rotation="0,30,-20", Scale="0.8,1.000,1") {
    rotationMode=BMM_Additive
    rotationSpace=BCS_BoneSpace
    scaleMode=BMM_Additive
    scaleSpace=BCS_BoneSpace
    translationMode=BMM_Additive
    translationSpace=BCS_BoneSpace
    alphaBoolBlend {blendOption=Linear}
    boneToModify {boneName=dyn_skirt_ft_l_AIM}
    rotation {roll=-20, yaw=30}
    scale {x=0.8}
   }
   n49 = "Transform (Modify) Bone - Bone: dyn_skirt_ft_r_AIM"(ComponentPose=@n48, Alpha=@n44, Translation="0,0,0", Rotation="0,0,0", Scale="0.8,1.000,1.000") {
    rotationMode=BMM_Additive
    rotationSpace=BCS_BoneSpace
    scaleMode=BMM_Additive
    scaleSpace=BCS_BoneSpace
    translationMode=BMM_Additive
    translationSpace=BCS_BoneSpace
    alphaBoolBlend {blendOption=Linear}
    boneToModify {boneName=dyn_skirt_ft_r_AIM}
    scale {x=0.8}
   }
   n50 = "Transform (Modify) Bone - Bone: dyn_skirt_ft_l_AIM"(ComponentPose=@n49, Alpha=Get thigh_rot_in_l() -> (thigh_rot_in_l:double), Translation="0,0,0", Rotation="0,-20,0", Scale="1,1,1") {
    rotationMode=BMM_Additive
    rotationSpace=BCS_BoneSpace
    scaleMode=BMM_Additive
    scaleSpace=BCS_BoneSpace
    translationMode=BMM_Additive
    translationSpace=BCS_BoneSpace
    alphaBoolBlend {blendOption=Linear}
    alphaScaleBiasClamp {bInterpResult=true}
    boneToModify {boneName=dyn_skirt_ft_l_AIM}
    rotation {yaw=-20}
   }
   n51 = "Transform (Modify) Bone - Bone: dyn_skirt_ft_r_AIM"(ComponentPose=@n50, Alpha=Get thigh_rot_in_r() -> (thigh_rot_in_r:double), Translation="0,0,0", Rotation="0,20,0", Scale="1,1,1") {
    rotationMode=BMM_Additive
    rotationSpace=BCS_BoneSpace
    scaleMode=BMM_Additive
    scaleSpace=BCS_BoneSpace
    translationMode=BMM_Additive
    translationSpace=BCS_BoneSpace
    alphaBoolBlend {blendOption=Linear}
    alphaScaleBiasClamp {bInterpResult=true}
    boneToModify {boneName=dyn_skirt_ft_r_AIM}
    rotation {yaw=20}
   }
   n52 = "Transform (Modify) Bone - Bone: dyn_skirt_ft_l_AIM"(ComponentPose=@n51, Alpha=Get thigh_rot_bk_l(), Translation="0,0,0", Rotation="25,10,0", Scale="1,1,1") {
    rotationMode=BMM_Additive
    rotationSpace=BCS_BoneSpace
    scaleMode=BMM_Additive
    scaleSpace=BCS_BoneSpace
    translationMode=BMM_Additive
    translationSpace=BCS_BoneSpace
    alphaBoolBlend {blendOption=Linear}
    boneToModify {boneName=dyn_skirt_ft_l_AIM}
    rotation {pitch=25, yaw=10}
   }
   n53 = "Transform (Modify) Bone - Bone: dyn_skirt_ft_r_AIM"(ComponentPose=@n52, Alpha=Get thigh_rot_bk_r(), Translation="0,0,0", Rotation="25,-10,0", Scale="1,1,1") {
    rotationMode=BMM_Additive
    rotationSpace=BCS_BoneSpace
    scaleMode=BMM_Additive
    scaleSpace=BCS_BoneSpace
    translationMode=BMM_Additive
    translationSpace=BCS_BoneSpace
    alphaBoolBlend {blendOption=Linear}
    boneToModify {boneName=dyn_skirt_ft_r_AIM}
    rotation {pitch=25, yaw=-10}
   }
   n54 = Get head_rot_bk() -> (head_rot_bk:double)
   n55 = "Transform (Modify) Bone - Bone: def_collar_flap_bk"(ComponentPose=@n53, Alpha=@n54, Translation="-1.773,-0.000,-1", Rotation="-20,0.000,0.000", Scale="1,1,1") {
    rotationMode=BMM_Additive
    rotationSpace=BCS_BoneSpace
    scaleMode=BMM_Additive
    scaleSpace=BCS_BoneSpace
    translationMode=BMM_Additive
    translationSpace=BCS_BoneSpace
    alphaBoolBlend {blendOption=Linear}
    boneToModify {boneName=def_collar_flap_bk}
    rotation {pitch=-20}
    translation {x=-1.773, z=-1}
   }
   n56 = "Transform (Modify) Bone - Bone: def_collar_bk"(ComponentPose=@n55, Alpha=@n54, Translation="0,0,0", Rotation="65,0.000,0.000", Scale="0.8,1.000,1.000") {
    rotationMode=BMM_Additive
    rotationSpace=BCS_BoneSpace
    scaleMode=BMM_Additive
    scaleSpace=BCS_BoneSpace
    translationMode=BMM_Additive
    translationSpace=BCS_BoneSpace
    alphaBoolBlend {blendOption=Linear}
    boneToModify {boneName=def_collar_bk}
    rotation {pitch=65}
    scale {x=0.8}
   }
   n57 = "Transform (Modify) Bone - Bone: def_collar_r"(ComponentPose=@n56, Alpha=@n54, Translation="0,0,0", Rotation="11.542,5.543,20.820", Scale="0.8,1,1") {
    rotationMode=BMM_Additive
    rotationSpace=BCS_BoneSpace
    scaleMode=BMM_Additive
    scaleSpace=BCS_BoneSpace
    translationMode=BMM_Additive
    translationSpace=BCS_BoneSpace
    alphaBoolBlend {blendOption=Linear}
    boneToModify {boneName=def_collar_r}
    rotation=<rot 11.542 5.543 20.8199>
    scale {x=0.8}
   }
   n58 = "Transform (Modify) Bone - Bone: def_collar_l"(ComponentPose=@n57, Alpha=@n54, Translation="0,0,0", Rotation="25.000,-0.000,-15.001", Scale="0.8,1,1") {
    rotationMode=BMM_Additive
    rotationSpace=BCS_BoneSpace
    scaleMode=BMM_Additive
    scaleSpace=BCS_BoneSpace
    translationMode=BMM_Additive
    translationSpace=BCS_BoneSpace
    alphaBoolBlend {blendOption=Linear}
    boneToModify {boneName=def_collar_l}
    rotation {pitch=25, roll=-15.001}
    scale {x=0.8}
   }
   n59 = Get head_rot_r() -> (head_rot_r:double)
   n60 = "Transform (Modify) Bone - Bone: def_collar_flap_bk"(ComponentPose=@n58, Alpha=@n59, Translation="0,0,0", Rotation="-0.069,-1.642,-9.863", Scale="1,1,1") {
    rotationMode=BMM_Additive
    rotationSpace=BCS_BoneSpace
    scaleMode=BMM_Additive
    scaleSpace=BCS_BoneSpace
    translationMode=BMM_Additive
    translationSpace=BCS_BoneSpace
    alphaBoolBlend {blendOption=Linear}
    boneToModify {boneName=def_collar_flap_bk}
    rotation=<rot -0.069 -1.642 -9.8632>
   }
   n61 = "Transform (Modify) Bone - Bone: def_collar_bk"(ComponentPose=@n60, Alpha=@n59, Translation="0,0,0", Rotation="-0.003,0.000,-15.000", Scale="1,1,1") {
    rotationMode=BMM_Additive
    rotationSpace=BCS_BoneSpace
    scaleMode=BMM_Additive
    scaleSpace=BCS_BoneSpace
    translationMode=BMM_Additive
    translationSpace=BCS_BoneSpace
    alphaBoolBlend {blendOption=Linear}
    boneToModify {boneName=def_collar_bk}
    rotation=<rot -0.003 0 -15.0002>
   }
   n62 = "Transform (Modify) Bone - Bone: def_collar_r"(ComponentPose=@n61, Alpha=@n59, Translation="0,0,0", Rotation="28.029,-7.404,-27.637", Scale="1,1,1") {
    rotationMode=BMM_Additive
    rotationSpace=BCS_BoneSpace
    scaleMode=BMM_Additive
    scaleSpace=BCS_BoneSpace
    translationMode=BMM_Additive
    translationSpace=BCS_BoneSpace
    alphaBoolBlend {blendOption=Linear}
    boneToModify {boneName=def_collar_r}
    rotation=<rot 28.0289 -7.4041 -27.6368>
   }
   n63 = "Transform (Modify) Bone - Bone: def_collar_flap_r"(
    ComponentPose=@n62
    Alpha=Get arm_rot_up_r() -> (arm_rot_up_r:double)
    Translation="-4.023,-0.110,0.207"
    Rotation="77.360,28.945,3.136"
    Scale="0.500,1.000,1.000"
   ) {
    rotationMode=BMM_Additive
    rotationSpace=BCS_BoneSpace
    scaleMode=BMM_Additive
    scaleSpace=BCS_BoneSpace
    translationMode=BMM_Additive
    translationSpace=BCS_BoneSpace
    alphaBoolBlend {blendOption=Linear}
    boneToModify {boneName=def_collar_flap_r}
    rotation=<rot 77.36 28.945 3.136>
    scale {x=0.5}
    translation=<-4.0229 -0.11 0.2071>
   }
   n64 = Anim Dynamics - Bone: dyn_collar_flap_bk(ComponentPose=@n63, Alpha=1, ExternalForce="0,0,0") {
    angularDampingOverride=0.9999
    angularSpringConstant=600
    bAngularSpring=true
    bGravityOverrideInSimSpace=true
    bOverrideAngularDamping=true
    bUseGravityOverride=true
    gravityScale=0.35
    simulationSpace=RootRelative
    alphaBoolBlend {blendOption=Linear}
    boundBone {boneName=dyn_collar_flap_bk}
    componentAppliedLinearAccClamp=<3000 3000 3000>
    componentLinearAccScale=<0.5 0.5 0.2>
    componentLinearVelScale=<0.5 0.5 0.2>
    gravityOverride {z=-980}
    physicsBodyDefinitions [{
      sphereCollisionRadius=0
      boundBone {boneName=dyn_collar_flap_bk}
      boxExtents=<10 20 15>
      constraintSetup {angularLimitsMax=<0 20 0>, angularLimitsMin=<0 -30 0>, angularTarget=<0 0 0>}
      localJointOffset=<3 0 0.5>
     }]
    relativeSpaceBone {boneName=root}
   }
   n65 = Anim Dynamics - Bone: dyn_collar_flap_l(ComponentPose=@n64, Alpha=1, ExternalForce="0,0,0") {
    angularDampingOverride=0.9999
    angularSpringConstant=600
    bAngularSpring=true
    bGravityOverrideInSimSpace=true
    bOverrideAngularDamping=true
    bUseGravityOverride=true
    gravityScale=0.35
    simulationSpace=RootRelative
    alphaBoolBlend {blendOption=Linear}
    boundBone {boneName=dyn_collar_flap_l}
    componentAppliedLinearAccClamp=<3000 3000 3000>
    componentLinearAccScale=<0.5 0.3 0.5>
    componentLinearVelScale=<0.5 0.3 0.5>
    gravityOverride {z=-980}
    physicsBodyDefinitions [{
      sphereCollisionRadius=0
      boundBone {boneName=dyn_collar_flap_l}
      boxExtents=<10 20 15>
      constraintSetup {angularLimitsMax=<0 15 2>, angularLimitsMin=<0 -35 -2>, angularTarget=<0 0 0>}
      localJointOffset=<3 0 0.5>
     }]
    relativeSpaceBone {boneName=root}
   }
   n66 = Anim Dynamics - Bone: dyn_collar_flap_r_01(ComponentPose=@n65, Alpha=1, ExternalForce="0,0,0") {
    angularDampingOverride=0.9999
    angularSpringConstant=600
    bAngularSpring=true
    bGravityOverrideInSimSpace=true
    bOverrideAngularDamping=true
    bUseGravityOverride=true
    gravityScale=0.35
    simulationSpace=RootRelative
    alphaBoolBlend {blendOption=Linear}
    boundBone {boneName=dyn_collar_flap_r_01}
    componentAppliedLinearAccClamp=<3000 3000 3000>
    componentLinearAccScale=<0.5 0.3 0.5>
    componentLinearVelScale=<0.5 0.3 0.5>
    gravityOverride {z=-980}
    physicsBodyDefinitions [{
      sphereCollisionRadius=0
      boundBone {boneName=dyn_collar_flap_r_01}
      boxExtents=<10 20 15>
      constraintSetup {angularLimitsMax=<0 15 2>, angularLimitsMin=<0 -35 -2>, angularTarget=<0 0 0>}
      localJointOffset=<3 0 0.5>
     }]
    relativeSpaceBone {boneName=root}
   }
   n67 = Anim Dynamics - Bone: dyn_collar_flap_r_02(ComponentPose=@n66, Alpha=1, ExternalForce="0,0,0") {
    angularDampingOverride=0.9999
    angularSpringConstant=600
    bAngularSpring=true
    bGravityOverrideInSimSpace=true
    bOverrideAngularDamping=true
    bUseGravityOverride=true
    gravityScale=0.35
    simulationSpace=RootRelative
    alphaBoolBlend {blendOption=Linear}
    boundBone {boneName=dyn_collar_flap_r_02}
    componentAppliedLinearAccClamp=<3000 3000 3000>
    componentLinearAccScale=<0.5 0.3 0.5>
    componentLinearVelScale=<0.5 0.3 0.5>
    gravityOverride {z=-980}
    physicsBodyDefinitions [{
      sphereCollisionRadius=0
      boundBone {boneName=dyn_collar_flap_r_02}
      boxExtents=<10 20 15>
      constraintSetup {angularLimitsMax=<0 15 2>, angularLimitsMin=<0 -35 -2>, angularTarget=<0 0 0>}
      localJointOffset=<3 0 0.5>
     }]
    relativeSpaceBone {boneName=root}
   }
   n68 = "Constraint - dyn_zipper_MSTR (PPPPP)"!(
    ComponentPose=@n67
    Alpha=1
    ConstraintWeights_0=0.0792
    ConstraintWeights_1:float=0.2065
    ConstraintWeights_2:float=0.3975
    ConstraintWeights_3:float=0.2413
    ConstraintWeights_4:float=0.0755
   ) -> (Pose) {
    alphaBoolBlend {blendOption=Linear}
    boneToModify {boneName=dyn_zipper_MSTR}
    constraintSetup [
     {targetBone {boneName=pelvis}}
     {targetBone {boneName=spine_01}}
     {targetBone {boneName=spine_02}}
     {targetBone {boneName=spine_03}}
     {targetBone {boneName=spine_04}}
    ]
    constraintWeights [1,1,1,1,1]
   }
   n69 = Anim Dynamics - Bone: dyn_zipper(ComponentPose=@n68, Alpha=1) {
    angularDampingOverride=0.999
    angularSpringConstant=1000
    bGravityOverrideInSimSpace=true
    bOverrideAngularDamping=true
    bUseGravityOverride=true
    gravityScale=0.35
    simulationSpace=RootRelative
    alphaBoolBlend {blendOption=Linear}
    boundBone {boneName=dyn_zipper}
    componentAppliedLinearAccClamp=<3000 3000 3000>
    componentLinearAccScale=<0.5 0.5 0.5>
    componentLinearVelScale=<0.5 0.5 0.5>
    gravityOverride {z=-980}
    physicsBodyDefinitions [{
      sphereCollisionRadius=0
      boundBone {boneName=dyn_zipper}
      boxExtents=<15 10 10>
      constraintSetup {angularLimitsMax=<0 10 45>, angularLimitsMin=<0 -45 -45>, angularTarget=<0 0 0>}
      localJointOffset=<5 0 0>
     }]
    relativeSpaceBone {boneName=spine_05}
   }
   n70 = "Transform (Modify) Bone - Bone: deform_armPad_r"(ComponentPose=@n69, Alpha=Get arm_rot_up_r(), Translation="1.746,0,0.378", Rotation="-3.000,0.000,5.000", Scale="1,1,1") {
    rotationMode=BMM_Additive
    rotationSpace=BCS_BoneSpace
    scaleMode=BMM_Additive
    scaleSpace=BCS_BoneSpace
    translationMode=BMM_Additive
    translationSpace=BCS_BoneSpace
    alphaBoolBlend {blendOption=Linear}
    boneToModify {boneName=deform_armPad_r}
    rotation {pitch=-3, roll=5}
    translation {x=1.746, z=0.378}
   }
   n71 = "Transform (Modify) Bone - Bone: deform_armPad_l"(ComponentPose=@n70, Alpha=Get arm_rot_up_l() -> (arm_rot_up_l:double), Translation="3,0,0.3", Rotation="-2,20,-3", Scale="1,1,1") {
    rotationMode=BMM_Additive
    rotationSpace=BCS_BoneSpace
    scaleMode=BMM_Additive
    scaleSpace=BCS_BoneSpace
    translationMode=BMM_Additive
    translationSpace=BCS_BoneSpace
    alphaBoolBlend {blendOption=Linear}
    boneToModify {boneName=deform_armPad_l}
    rotation=<rot -2 20 -3>
    translation {x=3, z=0.3}
   }
   Save cached pose 'prePose'(Pose=Component To Local(ComponentPose=@n71)) {cacheName=prePose}
  ]}
 EventGraph {nodes [
   Event BlueprintUpdateAnimation()
   > n1 = Sequence!(execute) -> (then_0, then_1, then_2:Exec, then_3:Exec, then_4:Exec, then_5:Exec)
   n2 = GetOwnerSkeletalMeshComponent()
   n3 = BreakRotator(InRot=BreakTransform(InTransform=GetDeltaTransformFromRefPose(self=@n2, BoneName=thigh_r, BaseName=pelvis)).Rotation)
   > Set thigh_rot_up_r(thigh_rot_up_r:double=MapRangeClamped(Value=@n3.Yaw, InRangeA=-15, InRangeB=-100)) -> (Output_Get:double)
   > Set thigh_rot_bk_r(thigh_rot_bk_r:double=MapRangeClamped(Value=@n3.Yaw, InRangeA=0, InRangeB=30)) -> (Output_Get:double)
   > Set thigh_rot_in_r(thigh_rot_in_r:double=MapRangeClamped(Value=@n3.Pitch, InRangeA=0, InRangeB=-30)) -> (Output_Get:double)
   > Set thigh_rot_out_r(thigh_rot_out_r:double=MapRangeClamped(Value=@n3.Pitch, InRangeA=0, InRangeB=20)) -> (Output_Get:double)
   n4 = GetOwnerSkeletalMeshComponent()
   n5 = BreakRotator(InRot=BreakTransform(InTransform=GetDeltaTransformFromRefPose(self=@n4, BoneName=thigh_l, BaseName=pelvis)).Rotation)
   > Set thigh_rot_up_l(thigh_rot_up_l:double=MapRangeClamped(Value=@n5.Yaw, InRangeA=-10, InRangeB=-100)) -> (Output_Get:double)
   > Set thigh_rot_bk_l(thigh_rot_bk_l:double=MapRangeClamped(Value=@n5.Yaw, InRangeA=0, InRangeB=30)) -> (Output_Get:double)
   > Set thigh_rot_in_l(thigh_rot_in_l:double=MapRangeClamped(Value=@n5.Pitch, InRangeA=0, InRangeB=-20)) -> (Output_Get:double)
   > Set thigh_rot_out_l(thigh_rot_out_l:double=MapRangeClamped(Value=@n5.Pitch, InRangeA=0, InRangeB=20)) -> (Output_Get:double)
   n6 = BreakTransform(InTransform=GetDeltaTransformFromRefPose(self=GetOwnerSkeletalMeshComponent(), BoneName=thigh_r, BaseName=calf_r))
   > Set knee_rot_r(knee_rot_r:double=MapRangeClamped(Value=BreakRotator(InRot=@n6.Rotation).Yaw, InRangeA=0, InRangeB=-90)) -> (Output_Get:double)
   n7 = Gravity Override State Function DEPRECATED(
    execute=@n1.then_1
    CharacterPartAnimInstance=Self-Reference()
    GameTimeInSeconds=GetGameTimeInSeconds()
    GravityOverrideParams=Get Gravity Override Params()
   )
   > Set GravityOverride(GravityOverride=@n7)
   n8 = Branch(execute=@n1.then_2, Condition=Get bIsCrouching())
   > n9 = Branch(Condition=Get bIsCrouchSprinting())
   Set backLoin_deform_alpha(execute=[@n9.then, @n8.else], backLoin_deform_alpha:double=0) -> (Output_Get:double)
   n10 = GetOwnerSkeletalMeshComponent()
   n11 = BreakRotator(InRot=BreakTransform(InTransform=GetDeltaTransformFromRefPose(self=@n10, BoneName=thigh_r, BaseName=pelvis)).Rotation)
   Set backLoin_deform_alpha(
    execute=@n9.else
    backLoin_deform_alpha=float * float(A=MapRangeClamped(Value=@n11.Yaw, InRangeA=0, InRangeB=-100, OutRangeA=1, OutRangeB=0), B=Get knee_rot_r() -> (knee_rot_r:double)) {is=title K2Node_PromotableOperator}
   )
   n12 = GetOwnerSkeletalMeshComponent()
   n13 = BreakRotator(InRot=BreakTransform(InTransform=GetDeltaTransformFromRefPose(self=@n12, BoneName=head, BaseName=spine_05)).Rotation)
   Set head_rot_bk(execute=@n1.then_3, head_rot_bk:double=MapRangeClamped(Value=@n13.Yaw, InRangeA=-20, InRangeB=-50)) -> (Output_Get:double)
   > Set head_rot_r(head_rot_r:double=MapRangeClamped(Value=@n13.Pitch, InRangeA=0, InRangeB=40)) -> (Output_Get:double)
   n14 = GetOwnerSkeletalMeshComponent()
   n15 = BreakRotator(InRot=BreakTransform(InTransform=GetDeltaTransformFromRefPose(self=@n14, BoneName=upperarm_r, BaseName=spine_05)).Rotation)
   > Set arm_rot_up_r(arm_rot_up_r:double=MapRangeClamped(Value="Absolute (Float)"(A=@n15.Yaw), InRangeA=0, InRangeB=135)) -> (Output_Get:double)
   n16 = GetOwnerSkeletalMeshComponent()
   n17 = BreakRotator(InRot=BreakTransform(InTransform=GetDeltaTransformFromRefPose(self=@n16, BoneName=upperarm_l, BaseName=spine_05)).Rotation)
   > Set arm_rot_up_l(arm_rot_up_l:double=MapRangeClamped(Value="Absolute (Float)"(A=@n17.Yaw), InRangeA=0, InRangeB=125)) -> (Output_Get:double)
   n18 = PlayerMovement(execute=@n1.then_4, AnimInstance=Self-Reference(), XDotProductCoefficient=1, YDotProductCoefficient=1, ZDotProductCoefficient=1)
   n19 = MapRangeClamped(Value=@n18.YMovement, InRangeA=0, InRangeB=1)
   n20 = float * float(A="Absolute (Float)"(A=@n18), B=0.5) {is=title K2Node_CommutativeAssociativeBinaryOperator}
   > Set BwdMovement(BwdMovement:double=MapRangeClamped(Value=float + float(A=@n20, B=@n19) {is=title K2Node_CommutativeAssociativeBinaryOperator}, InRangeA=0, InRangeB=1.1, OutRangeB=0.4)) -> (Output_Get:double)
   n21 = Branch(execute=@n1.then_5, Condition=Get bIsInFrontEnd())
   > Set BwdMovement(BwdMovement=0.1)
   n22 = Branch(execute=@n21.else, Condition=OR Boolean(A=Get bIsCrouching(), B=Get bIsSkyDiving()))
   > Set BwdMovement(BwdMovement=0)
   Set BwdMovement(execute=@n22.else, BwdMovement=Get BwdMovement())
   Event OnWeaponChanged()
   > n23 = Get Dynamic Material Instance(SkeletalMeshComponent=GetPartSkeletalMeshComponent())
   n24 = PickaxeItemDefinitionHasAnyMetaTag(Tags=Make Array("[0]"=(TagName="Cosmetics.MetaTags.LoneWolf")), bRequiresEquipped=true)
   > n25 = Branch(Condition=@n24)
   > n26 = Branch(Condition=Get bIsSkyDiving())
   SetScalarParameterValue(execute=[@n27.then, @n26.then, @n28.else, @n25.else], self=@n23, ParameterName=Hide Element 01, Value=0)
   SetScalarParameterValue(execute=[@n26.else, @n28.then], self=@n23, ParameterName=Hide Element 01, Value=1)
   n27 = Event OnBeginSkydiving()
   Event OnEndSkydiving()
   > n28 = Branch(Condition=@n24)
  ]}
]
variables [
 arm_rot_up_l {type=number}
 arm_rot_up_r {type=number}
 backLoin_deform_alpha {type=number}
 BwdMovement {type=number}
 Gravity Override Params {
  type=GravityOverrideParamsStruct
  default {skydive_FloatingDown {gravityOverride=<0 -700 100>, windAmplitude=<200 -500 200>, windOffset=<0 -600 200>}}
 }
 GravityOverride {type=Vector}
 head_rot_bk {type=number}
 head_rot_r {type=number}
 knee_rot_r {type=number}
 thigh_rot_bk_l {type=number}
 thigh_rot_bk_r {type=number}
 thigh_rot_in_l {type=number}
 thigh_rot_in_r {type=number}
 thigh_rot_out_l {type=number}
 thigh_rot_out_r {type=number}
 thigh_rot_up_l {type=number}
 thigh_rot_up_r {type=number}
]
